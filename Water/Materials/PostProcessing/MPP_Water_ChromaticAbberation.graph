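# Post-process material that adds chromatic aberration to water by splitting scene color: red stays put, green shifts once and blue twice as far.
# The offset follows the view-space refraction of the camera vector about the scene normal, via custom HLSL.
# It scales with blue-over-red color dominance and inverted Fresnel, shrinking with distance.
# A mask limits it to the water shading model ID or pixels below the WaterZ height.
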Material /Water/Materials/PostProcessing/MPP_Water_ChromaticAbberation
graph {
 properties {materialDomain=MD_PostProcess, pixelDepthOffsetMode=PDOM_Legacy, usedShadingModels=Default Lit}
 nodes [
  n1 = TextureCoordinate()
  n2 = ComponentMask(None=SceneTexture() {sceneTextureId=PPI_ShadingModelID}.Color) {r=true}
  n3 = Divide(A=ScalarParameter() {defaultValue=10, parameterName=Shading Model ID for Water}) {constB=255}
  n4 = ScalarParameter() {parameterName=WaterZ}
  n5 = ComponentMask(None=WorldPosition().XYZ) {b=true}
  n6 = LinearInterpolate(
   A=Saturate(None=MaterialFunctionCall!(Bit=@n2, BitMask=@n3) {materialFunction=/Engine/Functions/Engine_MaterialFunctions01/Texturing/BitMask}.Result)
   Alpha=Saturate(None=Subtract(A=@n4, B=@n5))
  )
  MaterialExpressionComponentMask_5 = ComponentMask(None=SceneTexture() {sceneTextureId=PPI_WorldNormal}.Color) {b=true, g=true, r=true}
  n7 = OneMinus(None=Fresnel(Normal=@MaterialExpressionComponentMask_5) {baseReflectFraction=0})
  n8 = SceneTexture() {bFiltered=true, sceneTextureId=PPI_PostProcessInput0}
  n9 = MaterialFunctionCall(Float3=@n8.Color) {materialFunction=&BreakOutFloat3Components}
  n10 = Subtract(A=@n9.B, B=Add(A=@n9.R, B=ScalarParameter() {parameterName=Chromatic Mask Bias}))
  n11 = ScalarParameter() {defaultValue=0.3, parameterName=Chromatic Color Mask Length}
  MaterialExpressionCameraVectorWS_0 = CameraVectorWS()
  MaterialExpressionScalarParameter_4 = ScalarParameter() {defaultValue=0.75, parameterName=n1 n2}
  n12 = Custom!(CameraVector=@MaterialExpressionCameraVectorWS_0, Normal=@MaterialExpressionComponentMask_5, n1n2=@MaterialExpressionScalarParameter_4) {
   code="refract(CameraVector, Normal, n1n2);"
   description=Refract
   inputs [
    {inputName=CameraVector, input {expression=&self:MaterialExpressionCameraVectorWS_0}}
    {inputName=Normal, input {expression=&self:MaterialExpressionComponentMask_5}}
    {inputName=n1n2, input {expression=&self:MaterialExpressionScalarParameter_4}}
   ]
  }
  n13 = Transform(None=@n12) {transformSourceType=TRANSFORMSOURCE_World, transformType=TRANSFORM_View}
  n14 = Multiply(
   A=Saturate(None=Divide(A=@n10, B=@n11))
   B=Multiply(A=ComponentMask(None=@n13) {g=true, r=true}, B=ScalarParameter() {defaultValue=7, parameterName=Chromatic Spead Texels})
  )
  n15 = Divide(
   A=ComponentMask(None=SceneTexture() {bFiltered=true, sceneTextureId=PPI_SceneDepth}.Color) {r=true}
   B=ScalarParameter() {defaultValue=1500, parameterName=Pixel Depth Max}
  )
  n16 = Multiply(
   A=Divide(A=Multiply(A=@n7, B=@n14), B=Max(A=@n15))
   B=ComponentMask(None=SceneTexture() {sceneTextureId=PPI_ShadingModelID}.InvSize) {g=true}
  )
  n17 = Add(A=@n1, B=Multiply(A=@n16, B=@n6))
  n18 = Add(A=@n17, B=Multiply(A=@n16, B=@n6))
  MaterialFunctionCall(
   X=ComponentMask(None=SceneTexture(UVs=@n1) {bFiltered=true, sceneTextureId=PPI_PostProcessInput0}.Color) {r=true}
   Y=ComponentMask(None=SceneTexture(UVs=@n17) {bFiltered=true, sceneTextureId=PPI_PostProcessInput0}.Color) {g=true}
   Z=ComponentMask(None=SceneTexture(UVs=@n18) {bFiltered=true, sceneTextureId=PPI_PostProcessInput0}.Color) {b=true}
  ) {materialFunction=&MakeFloat3}
  Normalize(VectorInput=@n8.Color)
 ]
}
