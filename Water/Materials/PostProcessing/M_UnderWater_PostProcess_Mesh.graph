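# Translucent unlit underwater effect material for a mesh, tinting scene color below the waterline.
# Scene color gets exponential depth absorption and fog, with Schlick-phase sky-light scattering.
# Iterated Gerstner wave lookups on a camera-relative near plane built from UVs find wave height, passed through a vertex interpolator to mask the fog and a colored waterline band.
# Caustics and shading-model mask branches do not feed that composite.
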Material /Water/Materials/PostProcessing/M_UnderWater_PostProcess_Mesh
graph {
 properties {
  bUseTranslucencyVertexFog=false
  blendMode=BLEND_Translucent
  pixelDepthOffsetMode=PDOM_Legacy
  shadingModel=MSM_Unlit
  translucencyLightingMode=TLM_SurfacePerPixelLighting
  twoSided=true
  usedShadingModels=Unlit
 }
 nodes [
  Constant3Vector() {constant=<rgba 0 1 0 0>}
  Constant()
  n1 = MaterialFunctionCall!(
   Cosine=DotProduct(A=SkyAtmosphereLightDirection(), B=CameraVectorWS())
   Anisotropy=ScalarParameter() {defaultValue=0.7, parameterName=Scattering Aniso}
  ) {materialFunction=/Engine/Functions/Engine_MaterialFunctions01/Volumetrics/SchlickPhase}
  n2 = VectorParameter() {parameterName=Fog Scatter Color, defaultValue=<rgba 0.5 1 0.75 1>}
  n3 = Multiply(A=@n2.RGB, B=@n2.RGB)
  n4 = VectorParameter() {parameterName=Fog Ambient Color, defaultValue=<rgba 0.1 0.4 0.75 0.15>}
  n5 = Min(A=SceneDepth().R, B=ScalarParameter() {defaultValue=250000, parameterName=Max Depth})
  n6 = VectorParameter() {parameterName=Absorption, defaultValue=<rgba 25 5 2 2500>}
  MaterialExpressionMultiply_11 = Multiply(A=@n5, B=Divide(A=@n6.RGB, B=@n6.RGB))
  n7 = Custom!(x=@MaterialExpressionMultiply_11) {
   code="exp(-x);"
   description=exp
   inputs [{inputName=x, input {expression=&self:MaterialExpressionMultiply_11}}]
  }
  MaterialExpressionMultiply_12 = Multiply(A=@n5, B=ScalarParameter() {defaultValue=0.001, parameterName=Fog})
  n8 = Custom(x=@MaterialExpressionMultiply_12) {
   code="exp(-x);"
   description=exp
   outputType=CMOT_Float1
   inputs [{inputName=x, input {expression=&self:MaterialExpressionMultiply_12}}]
  }
  n9 = LinearInterpolate(A=Add(A=Multiply(A=@n1.Result, B=@n3), B=Multiply(A=@n4.RGB, B=@n4.RGB)), B=Multiply(A=@n7, B=SceneColor!(UVs).RGB), Alpha=@n8)
  n10 = MaterialFunctionCall!(Position) {materialFunction=/Water/Materials/Functions/WaterBodyData}
  n11 = ScalarParameter() {defaultValue=0.1, parameterName=Wet Alpha}
  n12 = Add(A=@n10.Water Z Height, B=@n11)
  n13 = MaterialFunctionCall() {materialFunction=/Water/Materials/Functions/GetWaterTime}
  n14 = Transform(None=Constant3Vector() {constant=<rgba 0 0 1 0>}) {transformSourceType=TRANSFORMSOURCE_View}
  n15 = ScalarParameter() {defaultValue=10.005, parameterName=Near Z}
  n16 = MaterialFunctionCall(Float2=Subtract(A=TextureCoordinate()) {constB=0.5}) {materialFunction=&BreakOutFloat2Components}
  n17 = Multiply(A=@n16.R, B=Transform(None=Constant3Vector() {constant=<rgba 0 -1 0 0>}) {transformSourceType=TRANSFORMSOURCE_View})
  n18 = Multiply(A=@n16.G, B=Transform(None=Constant3Vector() {constant=<rgba 2 0 0 0>}) {transformSourceType=TRANSFORMSOURCE_View})
  n19 = Add(
   A=Add(A=Multiply(A=@n14, B=@n15), B=CameraPositionWS())
   B=Multiply(A=Add(A=@n17, B=@n18), B=ScalarParameter() {defaultValue=12, parameterName=Size})
  )
  n20 = MaterialFunctionCall!(WaveParameterRT, Override Position=@n19, Time=@n13.WaterTime) {materialFunction=/Water/Caustics/Materials/Functions/GetGerstnerWaves}
  n21 = Subtract(
   A=@n19
   B=MaterialFunctionCall(Override Position=Subtract(A=@n19, B=@n20.WPO), Time=@n13.WaterTime) {materialFunction=&GetGerstnerWaves}.WPO
  )
  n22 = MaterialFunctionCall(
   Override Position=Subtract(
    A=@n19
    B=MaterialFunctionCall(Override Position=@n21, Time=@n13.WaterTime) {materialFunction=&GetGerstnerWaves}.WPO
   )
   Time=@n13.WaterTime
  ) {materialFunction=&GetGerstnerWaves}
  n23 = ComponentMask(None=MaterialFunctionCall(Override Position=Subtract(A=@n19, B=@n22.WPO), Time=@n13.WaterTime) {materialFunction=&GetGerstnerWaves}.WPO) {b=true}
  n24 = Subtract(A=ComponentMask(None=WorldPosition().XYZ) {b=true}, B=ComponentMask(None=VertexInterpolator(VS=@n23).PS) {r=true})
  n25 = Saturate(None=Divide(A=Subtract(A=@n12, B=@n24), B=@n11))
  n26 = VectorParameter() {parameterName=Band Color, defaultValue=<rgba 0.01 0.2 0.3 0>}
  n27 = Divide(A=@n24, B=ScalarParameter() {defaultValue=1, parameterName=Band Width})
  n28 = ScalarParameter() {defaultValue=0.1, parameterName=Band Offset}
  n29 = LinearInterpolate(A=Multiply(A=@n26.RGB) {constB=0.5}, B=Multiply(A=@n26.RGB) {constB=2}, Alpha=Saturate(None=Add(A=@n27, B=@n28)))
  n30 = Transform(None=Constant3Vector() {constant=<rgba 0 0 1 0>}) {transformSourceType=TRANSFORMSOURCE_View}
  n31 = Multiply(
   A=ScalarParameter() {defaultValue=1, parameterName=Wet Radius}
   B=MaterialFunctionCall(Vector 2=ComponentMask(None=@n30) {g=true, r=true}) {materialFunction=/Engine/Functions/Engine_MaterialFunctions02/Utility/VectorLength}.V2 Length
  )
  n32 = ScalarParameter() {defaultValue=0.25, parameterName=Wet Hardness}
  LinearInterpolate(
   A=LinearInterpolate(A=SceneColor().RGB, B=Multiply(A=@n9), Alpha=Multiply(A=@n25)) {constA=1}
   B=Multiply(
    A=ComponentMask(None=@n29) {b=true, g=true, r=true}
    B=VectorParameter() {parameterName=Band Coloration Adjustment, defaultValue=<rgba 0.5 0.5 0.5 0>}.RGB
   )
   Alpha=Multiply(A=SphereMask(A=@n24, B=@n10.Water Z Height, Radius=@n31, Hardness=@n32), B=ScalarParameter() {defaultValue=1, parameterName=Band Opacity})
  )
  LinearInterpolate()
  n33 = MaterialFunctionCall!(
   Bit=ComponentMask(None=SceneTexture() {sceneTextureId=PPI_ShadingModelID}.Color) {r=true}
   BitMask=Divide(A=ScalarParameter() {defaultValue=10, parameterName=Shading Model ID for Water}) {constB=255}
  ) {materialFunction=/Engine/Functions/Engine_MaterialFunctions01/Texturing/BitMask}
  n34 = ComponentMask(None=SceneTexture() {sceneTextureId=PPI_WorldNormal}.Color) {b=true}
  LinearInterpolate(A=OneMinus(None=Saturate(None=@n33.Result)), Alpha=Saturate(None=Ceil(None=@n34)))
  MaterialFunctionCall() {materialFunction=&WorldPositionBehindTranslucency}
  n35 = Divide(
   A=MaterialFunctionCall!(ScreenUV=ScreenPosition().ViewportUV) {materialFunction=/Water/Materials/Functions/WorldPositionBehindWater}.Result
   B=ScalarParameter() {defaultValue=1000, parameterName=Caustics Scale}
  )
  Multiply(
   A=TextureSample(UVs=ComponentMask(None=@n35) {g=true, r=true}) {samplerType=SAMPLERTYPE_LinearColor, texture=/Water/Textures/Caustics/Caustics_Animated_01_HDR}.RGB
   B=ScalarParameter() {defaultValue=16, parameterName=Boost}
  )
  n36 = WorldPosition()
  Multiply(A=Add(A=Multiply(A=@n36.XYZ) {constB=-1}, B=@n19), B=ScalarParameter() {defaultValue=1, parameterName=WWW})
  Multiply()
  ScalarParameter() {parameterName=Wet Location}
  ScalarParameter() {parameterName=ref}
  ScalarParameter() {defaultValue=1}
 ]
}
