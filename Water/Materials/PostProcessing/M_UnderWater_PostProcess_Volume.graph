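# Post-process material that renders the underwater view and waterline band for a water body.
# It rebuilds each pixel's near-plane position, then compares it to water height from WaterBodyData plus Gerstner waves damped by flow velocity.
# That signed distance blends scene color into depth fog, built from exponential absorption and Schlick-phase scattering, and places a lit band at the surface.
# ES3_1 clips fog depth at the surface; a static switch overlays WaterBodyIndex for debugging.
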Material /Water/Materials/PostProcessing/M_UnderWater_PostProcess_Volume
graph {
 properties {
  bUseTranslucencyVertexFog=false
  blendMode=BLEND_Translucent
  floatPrecisionMode=MFPM_Full
  materialDomain=MD_PostProcess
  pixelDepthOffsetMode=PDOM_Legacy
  shadingModel=MSM_Unlit
  translucencyLightingMode=TLM_SurfacePerPixelLighting
  twoSided=true
  usedShadingModels=Unlit
 }
 nodes [
  Constant()
  LinearInterpolate()
  n1 = ScalarParameter() {parameterName=WaterBodyIndex}
  n2 = MaterialFunctionCall() {materialFunction=/Water/Materials/Functions/GetWaterTime}
  n3 = CameraVectorWS()
  n4 = DotProduct(A=@n3, B=MaterialFunctionCall() {materialFunction=&CameraDirectionVector}.Result)
  n5 = Multiply(A=Divide(A=@n3, B=Abs(None=@n4)), B=ScalarParameter() {defaultValue=10, parameterName=Near Plane})
  n6 = NamedRerouteDeclaration(None=Add(A=CameraPositionWS(), B=Multiply(A=@n5) {constB=-1}) {desc=Screen World Pos}) {name=Adjusted View Pos, nodeColor=<rgba 0 1 0.6016 1>}
  n7 = MaterialFunctionCall!(Position=@n6) {materialFunction=/Water/Materials/Functions/WaterBodyData}
  n8 = Divide(
   A=MaterialFunctionCall(Vector 2=@n7.Velocity) {materialFunction=/Engine/Functions/Engine_MaterialFunctions02/Utility/VectorLength}.V2 Length
   B=ScalarParameter() {defaultValue=100, parameterName=Velocity to Mask waves}
  )
  n9 = Multiply(
   A=ComponentMask(None=MaterialFunctionCall!(Water Body Index=@n1, Override Position=@n6, Time=@n2.WaterTime, Use Depth Attenuation) {materialFunction=/Water/Materials/Functions/ComputeGerstnerWaves}.WPO) {b=true}
   B=OneMinus(None=Saturate(None=@n8))
  )
  n10 = NamedRerouteDeclaration(None=@n7.Water Z Height) {name=Water Z Height At Ajusted View Pos, nodeColor=<rgba 0 1 0.3438 1>}
  n11 = ComponentMask(None=@n6) {b=true, desc=Camera Nearplane Z Pos}
  n12 = Subtract(A=NamedRerouteDeclaration(None=Add(A=@n9, B=@n10)) {name=Water Surface Z Pos, nodeColor=<rgba 0 1 0.25 1>}, B=@n11)
  n13 = Multiply(A=@n12) {constB=-1}
  n14 = ScalarParameter() {defaultValue=1, parameterName=Band Width}
  n15 = Saturate(None=Add(A=Divide(A=@n13, B=@n14), B=ScalarParameter() {defaultValue=0.5, parameterName=Band Offset}))
  n16 = SceneTexture() {sceneTextureId=PPI_PostProcessInput0}
  n17 = MaterialFunctionCall!(
   Cosine=DotProduct(A=SkyAtmosphereLightDirection(), B=CameraVectorWS())
   Anisotropy=ScalarParameter() {defaultValue=0.2, parameterName=Scattering Aniso}
  ) {materialFunction=/Engine/Functions/Engine_MaterialFunctions01/Volumetrics/SchlickPhase}
  n18 = VectorParameter() {parameterName=Fog Scatter Color, defaultValue=<rgba 0.5 1 0.75 3>}
  n19 = Multiply(A=@n18.RGB, B=@n18.RGB)
  n20 = VectorParameter() {parameterName=Fog Ambient Color, defaultValue=<rgba 0.1 0.4 0.75 0.25>}
  n21 = ComponentMask(None=SceneTexture() {sceneTextureId=PPI_SceneDepth}.Color) {r=true}
  n22 = NamedRerouteUsage()
  n23 = ComponentMask(None=NamedRerouteUsage()) {b=true}
  n24 = Subtract(A=@n22, B=@n23)
  n25 = Multiply(A=CameraVectorWS()) {constB=-1}
  n26 = Max(A=ComponentMask(None=NamedRerouteDeclaration(None=@n25) {name=Camera To Pixel Direction, nodeColor=<rgba 0 1 0.0859 1>}) {b=true}) {constB=0.0001}
  n27 = Min(A=If(A=@n22, B=@n23, "A > B"=Divide(A=@n24, B=@n26), "A < B"=Constant()), B=@n21)
  n28 = Min(
   A=NamedRerouteDeclaration(None=FeatureLevelSwitch(Default=@n21, ES3_1=@n27)) {name=Scene Depth, nodeColor=<rgba 0.2656 0 1 1>}
   B=ScalarParameter() {defaultValue=250000, parameterName=Max Depth}
  )
  n29 = VectorParameter() {parameterName=Absorption, defaultValue=<rgba 25 5 2 7500>}
  MaterialExpressionMultiply_11 = Multiply(A=@n28, B=Divide(A=@n29.RGB, B=@n29.RGB))
  n30 = Custom!(x=@MaterialExpressionMultiply_11) {
   code="exp(-x);"
   description=exp
   inputs [{inputName=x, input {expression=&self:MaterialExpressionMultiply_11}}]
  }
  n31 = SceneTexture() {sceneTextureId=PPI_PostProcessInput0}
  MaterialExpressionMultiply_12 = Multiply(A=@n28, B=ScalarParameter() {defaultValue=0.0002, parameterName=Fog})
  n32 = Custom(x=@MaterialExpressionMultiply_12) {
   code="exp(-x);"
   description=exp
   outputType=CMOT_Float1
   inputs [{inputName=x, input {expression=&self:MaterialExpressionMultiply_12}}]
  }
  n33 = LinearInterpolate(
   A=Add(A=Multiply(A=@n17.Result, B=@n19), B=Multiply(A=@n20.RGB, B=@n20.RGB))
   B=Multiply(A=@n30, B=ComponentMask(None=@n31.Color) {b=true, g=true, r=true})
   Alpha=@n32
  )
  n34 = ScalarParameter() {defaultValue=0.2, parameterName=Wet Alpha}
  n35 = Saturate(None=Divide(A=Add(A=@n12, B=@n34), B=@n34))
  n36 = VectorParameter() {parameterName=Band Color, defaultValue=<rgba 0.01 0.2 0.3 8>}
  n37 = Multiply(A=@n36.RGB, B=@n36.RGB)
  n38 = LinearInterpolate(A=Multiply(A=@n37) {constB=0.5}, B=Multiply(A=@n37) {constB=2}, Alpha=@n15)
  n39 = DotProduct(A=SkyAtmosphereLightDirection(), B=CameraVectorWS())
  n40 = ScalarParameter() {defaultValue=0.2, parameterName=Scattering Aniso}
  n41 = ScalarParameter() {parameterName=Wet Location}
  n42 = Transform(None=Constant3Vector() {constant=<rgba 0 0 1 0>}) {transformSourceType=TRANSFORMSOURCE_View}
  n43 = Multiply(
   A=ScalarParameter() {defaultValue=1, parameterName=Wet Radius}
   B=MaterialFunctionCall(Vector 2=ComponentMask(None=@n42) {g=true, r=true}) {materialFunction=&VectorLength}.V2 Length
  )
  n44 = ScalarParameter() {defaultValue=0.25, parameterName=Wet Hardness}
  n45 = LinearInterpolate(
   A=LinearInterpolate(A=ComponentMask(None=@n16.Color) {b=true, g=true, r=true}, B=Multiply(A=@n33), Alpha=Multiply(A=@n35)) {constA=1}
   B=Multiply(
    A=ComponentMask(None=@n38) {b=true, g=true, r=true}
    B=MaterialFunctionCall(Cosine=@n39, Anisotropy=@n40) {materialFunction=&SchlickPhase}.Result
   )
   Alpha=Multiply(A=SphereMask(A=@n12, B=@n41, Radius=@n43, Hardness=@n44), B=ScalarParameter() {defaultValue=1, parameterName=Band Opacity})
  )
  LinearInterpolate(A=@n15, B=@n45)
  ScalarParameter() {defaultValue=1}
  n46 = ScalarParameter() {parameterName=WaterBodyIndex}
  n47 = ScreenPosition()
  StaticSwitch(
   True=Add(
    A=@n45
    B=MaterialFunctionCall(Number=@n46, UVs=@n47.ViewportUV) {materialFunction=&DebugScalarValues}.Result
   )
   False=@n45
   Value=StaticBoolParameter() {parameterName=Debug WaterBodyIndex}
  )
  Subtract(A=@n11, B=@n9)
  WorldPosition()
 ]
}
