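; Blueprint actor that displays an underwater post-process effect through a static mesh material, with an Enable flag controlling mesh visibility.
; Its construction script calls MakeMID, which puts a dynamic instance of M_UnderWater_PostProcess_Mesh on slot 0.
; GetWaterHeightTexture rebuilds that instance and feeds it the first WaterZone's water info texture array as WaterVelocityAndHeight.
; Nothing here calls that function, and the BeginPlay, overlap and Tick events are empty.
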(event EventBeginPlay)

(event Collision|EventActorBeginOverlap
  (params (OtherActor :type Actor%20Object%20Reference)))

(event EventTick
  (params (DeltaSeconds :type Float%20%28single-precision%29)))

(fn GetWaterHeightTexture
  (CallFunction|MakeMID)
  (bind _outactors (Actor|GetAllActorsOfClass "/Script/Water.WaterZone"))
  (Rendering|Material|SetTextureParameterValue (Variables|Default|GetMID) "WaterVelocityAndHeight" (Variables|/Script/Water.WaterZone|GetWaterInfoTextureArray (Utilities|Array|Get(acopy) _outactors))))

(fn MakeMID
  (Rendering|CreateAndAssignMID (Variables|Default|GetMID) "/Water/Materials/PostProcessing/M_UnderWater_PostProcess_Mesh.M_UnderWater_PostProcess_Mesh"
    (:Out
      (Rendering|Material|SetMaterial (Variables|Default|GetStaticMesh) 0 (Variables|Default|GetMID)))))

(fn ConstructionScript
  (CallFunction|MakeMID)
  (Rendering|SetVisibility (Variables|Default|GetStaticMesh) (Variables|Default|GetEnable) false))
